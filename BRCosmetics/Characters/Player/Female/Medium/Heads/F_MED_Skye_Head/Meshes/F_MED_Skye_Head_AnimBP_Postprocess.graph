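# Post-process AnimBP for the Skye head that applies bone corrections for jaw opening and a worn mask.
# On customization load, it flags a mask when the Face part's asset name contains "Mask".
# With a mask outside emotes, it counter-rotates the head against its neck_02-relative delta, by default only in yaw.
# Additive offsets move three under-chin bones by the jawOpen curve and the center chin bone while masked.
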AnimBlueprint /BRCosmetics/Characters/Player/Female/Medium/Heads/F_MED_Skye_Head/Meshes/F_MED_Skye_Head_AnimBP_Postprocess
target_skeleton=/Game/Characters/Player/Common/Fortnite_Base_Head/Fortnite_Base_3L_Head_Skeleton
graphs [
 AnimGraph {nodes [
   n1 = "Transform (Modify) Bone - Bone: head"(
    ComponentPose=Local To Component(LocalPose=Input Pose())
    Alpha=1
    Translation="0,0,0"
    Rotation=Get Head_Rot() -> (Head_Rot:Rotator)
    Scale="1,1,1"
   ) {rotationMode=BMM_Additive, rotationSpace=BCS_BoneSpace, alphaBoolBlend {blendOption=Linear}, boneToModify {boneName=head}}
   n2 = Get JawOpen() -> (JawOpen:double)
   n3 = "Transform (Modify) Bone - Bone: FACIAL_C_UnderChin"(ComponentPose=@n1, Alpha=@n2, Translation="0,-2,-2", Rotation="0,0,0", Scale="1,1,1") {
    translationMode=BMM_Additive
    translationSpace=BCS_BoneSpace
    alphaBoolBlend {blendOption=Linear}
    boneToModify {boneName=FACIAL_C_UnderChin}
    translation {y=-2, z=-2}
   }
   n4 = "Transform (Modify) Bone - Bone: FACIAL_R_UnderChin"(ComponentPose=@n3, Alpha=@n2, Translation="0,-3,-2", Rotation="0,0,0", Scale="1,1,1") {
    translationMode=BMM_Additive
    translationSpace=BCS_BoneSpace
    alphaBoolBlend {blendOption=Linear}
    boneToModify {boneName=FACIAL_R_UnderChin}
    translation {y=-3, z=-2}
   }
   n5 = "Transform (Modify) Bone - Bone: FACIAL_L_UnderChin"(ComponentPose=@n4, Alpha=@n2, Translation="0,-3,-2", Rotation="0,0,0", Scale="1,1,1") {
    translationMode=BMM_Additive
    translationSpace=BCS_BoneSpace
    alphaBoolBlend {blendOption=Linear}
    boneToModify {boneName=FACIAL_L_UnderChin}
    translation {y=-3, z=-2}
   }
   n6 = "Transform (Modify) Bone - Bone: FACIAL_C_Chin"!(ComponentPose=@n5, bAlphaBoolEnabled:bool=Get isHelmet_On() -> (isHelmet_On:bool), Translation="0,-1,-1", Rotation="0,0,0", Scale="1,1,1") -> (Pose) {
    alphaInputType=Bool
    translationMode=BMM_Additive
    translationSpace=BCS_BoneSpace
    alphaBoolBlend {blendOption=Linear}
    boneToModify {boneName=FACIAL_C_Chin}
    translation {y=-1, z=-1}
   }
   Output Pose(Result=Component To Local(ComponentPose=@n6))
  ]}
 EventGraph {nodes [
   Event BlueprintUpdateAnimation()
   n1 = GetPartSkeletalMeshComponent()
   > Set JawOpen(JawOpen:double=GetCurveValue(self=GetAnimInstance(self=@n1), CurveName=CTRL_expressions_jawOpen)) -> (Output_Get:double)
   n2 = Get bIsPlayingEmote()
   > Branch(Condition=AND Boolean(A=Get isHelmet_On(), B=NOT Boolean(A=@n2)))
   n3 = GetOwnerSkeletalMeshComponent()
   n4 = BreakRotator(InRot=BreakTransform(InTransform=GetDeltaTransformFromRefPose(self=@n3, BoneName=head, BaseName=neck_02)).Rotation)
   n5 = Get Z_Percentage() -> (Z_Percentage:double)
   n6 = Get X_Percentage() -> (X_Percentage:double)
   n7 = Get Y_Percentage() -> (Y_Percentage:double)
   > Set Head_Rot(Head_Rot:Rotator=MakeRotator(
     Roll=float * float!(A=@n4, B=-1, C:double=@n6) -> (ReturnValue) {is=title K2Node_PromotableOperator}
     Pitch=float * float(A=@n4.Pitch, B=-1, C=@n7)
     Yaw=float * float(A=@n4.Yaw, B=-1, C=@n5)
    )) -> (Output_Get:Rotator)
   n8 = Event OnCharacterCustomizationLoaded()
   n9 = Get Array Item with EFort Custom Part Type(Array:Array of Skinned Mesh Component Object References=@n8.CustomMeshParts, PartType=Face) -> (Output:Skinned Mesh Component Object Reference)
   > Is Valid(InputObject=@n9)
   n10 = GetSkinnedAsset!(self:Skinned Mesh Component Object Reference=@n9) -> (ReturnValue:Skinned Asset Object Reference)
   > Set isHelmet_On(isHelmet_On:bool=Contains(SearchIn=GetDisplayName(Object=@n10), Substring=Get isHelmet() -> (isHelmet:String), bSearchFromEnd=true)) -> (Output_Get:bool)
   CharacterItemDefinitionHasAnyMetaTag(Tags=Make Array("[0]"=(TagName="Cosmetics.MetaTags.RoundThumb_Gumbo")))
  ]}
]
variables [
 Head_Rot {type=Rotator}
 isHelmet {type=string, default=Mask}
 isHelmet_On {type=boolean}
 JawOpen {type=number}
 X_Percentage {type=number}
 Y_Percentage {type=number}
 Z_Percentage {type=number, default=0.7}
]
